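# Material function, exposed to the library, that calculates the angle a point light's spherical source radius subtends from the shaded world position.
# It divides Source Radius by the distance from Light Position to World Position, with that distance clamped to at least the radius so the sine stays within one.
# A Custom HLSL node applies asin to give the Radians output, which a constant multiply converts for the Degrees output.
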MaterialFunction /Engine/Functions/Engine_MaterialFunctions01/Shading/CalcLightsourceAngle
graph {
 properties {
  bExposeToLibrary=true
  description=This function can be used to calculate the light source angle of a pointlight given a certain radius in world space.
  libraryCategoriesText [Misc,Shading]
 }
 nodes [
  n1 = FunctionInput() {
   bUsePreviewValueAsDefault=true
   description=This is the size of the light source radius in world units.
   inputName=Source Radius
   inputType=FunctionInput_Scalar
   sortPriority=1
   previewValue=<100 0 0 1>
  }
  n2 = FunctionInput() {
   description=The light position to calculate light source angle for. For a directional light the light source angle is always the same.
   inputName=Light Position
  }
  n3 = WorldPosition()
  MaterialExpressionDivide_6 = Divide(A=@n1, B=Max(A=@n1, B=Distance(A=@n2, B=@n3.XYZ))) {desc="sin(x)"}
  n4 = Custom!(x=@MaterialExpressionDivide_6) {
   code="return asin(x);"
   description="asin(x)"
   outputType=CMOT_Float1
   inputs [{inputName=x, input {expression=&self:MaterialExpressionDivide_6}}]
  }
  FunctionOutput(None=Multiply(A=FunctionOutput(None=@n4) {outputName=Radians, sortPriority=1}, B=Constant() {r=57.2958}) {desc=LSA}) {outputName=Degrees}
 ]
}
