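# Material function that computes how much an occluder sphere blocks a light cone, intended for baking static shading textures rather than realtime use.
# Three angles, in degrees or optionally radians, select between zero, smaller-cap area for containment, and partial overlap.
# Partial overlap uses Custom HLSL, either an analytic spherical cap intersection or a smoothstep approximation.
# It outputs intersection area and illumination, one minus that area over the light cap's area.
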MaterialFunction /Engine/Functions/Engine_MaterialFunctions01/Shading/Spherical-Cap-Intersection
graph {
 properties {
  bExposeToLibrary=true
  description="Mathematically computes the area of two spherecap intersections. This will return the occlusion ratio given a sphere occluder and a light angle. Returns the inverse, so it returns 1 when there was no occlusion of the light cone and 0 when there is full occlusion of the light cone. Units are specified in degrees. Note that this function is used to generate static textures for use in final shading, not meant to be used in realtime."
  libraryCategoriesText [Misc,Shading]
 }
 nodes [
  n1 = FunctionInput() {
   description="This is the angle of occlusion from the occluder. As the occluder gets closer to the camera, occlusion reaches 1. As the occluder gets infinitely far from the camera, the occlusion approaches 0. In Degrees."
   inputName=View Occlusion Angle
   inputType=FunctionInput_Scalar
   sortPriority=1
  }
  n2 = StaticBool()
  n3 = FunctionInput(Preview=@n2) {
   bUsePreviewValueAsDefault=true
   description=By default the angles are read as degrees. This option lets you specify the angles in radians to save a few instructions or if you are just hardcore.
   inputName=Use Radians
   inputType=FunctionInput_StaticBool
   sortPriority=5
  }
  n4 = StaticSwitch(True=Constant() {r=1}, False=Constant() {r=57.2958}, Value=@n3)
  MaterialExpressionDivide_17 = Divide(A=@n1, B=@n4) {desc=r0}
  n5 = FunctionInput() {
   description="If you only care about the spherecap intersection area, it does not matter if you consider this a light vector or just one of the two cone angles. When using the illumination output, this cone determines the illumination fraction. A larger source angle causes a blurrier penumbra. Specified in Degrees."
   inputName=Light Source Angle
   inputType=FunctionInput_Scalar
   sortPriority=2
   previewValue=<5 0 0 1>
  }
  MaterialExpressionDivide_18 = Divide(A=@n5, B=@n4) {desc=r1}
  n6 = Min(A=@MaterialExpressionDivide_17, B=@MaterialExpressionDivide_18)
  n7 = FunctionInput() {
   description=This is the angle between the vector to the occluder center and the light vector. In Degrees.
   inputName=Angle Between Cones
   inputType=FunctionInput_Scalar
  }
  MaterialExpressionDivide_19 = Divide(A=@n7, B=@n4) {desc=d}
  n8 = Custom!(d=@MaterialExpressionDivide_19, r0=@MaterialExpressionDivide_17, r1=@MaterialExpressionDivide_18) {
   code="float a = ((2*PI)-(2*PI*cos(min(r1,r0))));\r\na*=smoothstep(0,1,1-(d-length(r0-r1))/(r0+r1-(length(r0-r1))));\r\n\r\nreturn a;"
   description=Smoothstep version
   outputType=CMOT_Float1
   inputs [
    {inputName=d, input {expression=&self:MaterialExpressionDivide_19}}
    {inputName=r0, input {expression=&self:MaterialExpressionDivide_17}}
    {inputName=r1, input {expression=&self:MaterialExpressionDivide_18}}
   ]
  }
  n9 = Custom(d=@MaterialExpressionDivide_19, r0=@MaterialExpressionDivide_17, r1=@MaterialExpressionDivide_18) {
   code="float div1=cos(d)-(cos(r0)*cos(r1));\r\ndiv1/=sin(r0)*sin(r1);\r\n\r\nfloat div2=((-1)*cos(r1))+(cos(d)*cos(r0));\r\ndiv2/=sin(d)*sin(r0);\r\n\r\nfloat div3=((-1)*cos(r0))+(cos(d)*cos(r1));\r\ndiv3/=sin(d)*sin(r1);\r\n\r\nfloat term=(2*PI)-(2*PI*cos(r0));\r\nterm-=2*PI*cos(r1);\r\nterm-=2*acos(div1);\r\nterm+=2*cos(r0)*acos(div2);\r\nterm+=2*cos(r1)*acos(div3);\r\n\r\n\r\n\r\nreturn term;"
   description=SphereCap Intersection Area
   outputType=CMOT_Float1
   inputs [
    {inputName=d, input {expression=&self:MaterialExpressionDivide_19}}
    {inputName=r0, input {expression=&self:MaterialExpressionDivide_17}}
    {inputName=r1, input {expression=&self:MaterialExpressionDivide_18}}
   ]
  }
  n10 = FunctionInput(Preview=@n2) {
   bUsePreviewValueAsDefault=true
   description="When True, area intersection will be approximated using smoothstep instead of being calculated analytically. Smoothstep gives cleaner results with small light source angles (less than 1.0)."
   inputName=Use Smoothstep
   inputType=FunctionInput_StaticBool
   sortPriority=4
  }
  n11 = Constant()
  n12 = Multiply(
   A=OneMinus(None=Cosine(None=@n6) {period=6.2832})
   B=MaterialFunctionCall(Multiplier=Constant() {r=2}) {materialFunction=&Pi}.Result
  ) {desc=Single Cone intersection area}
  n13 = If(
   A=@n6
   B=Subtract(A=Max(A=@MaterialExpressionDivide_17, B=@MaterialExpressionDivide_18), B=@MaterialExpressionDivide_19)
   "A > B"=If(
    A=Add(A=@MaterialExpressionDivide_17, B=@MaterialExpressionDivide_18)
    B=@MaterialExpressionDivide_19
    "A > B"=StaticSwitch(True=@n8, False=@n9, Value=@n10)
    "A == B"=@n11
    "A < B"=@n11
   )
   "A == B"=@n12
   "A < B"=@n12
  )
  n14 = Multiply(
   A=OneMinus(None=Cosine(None=@MaterialExpressionDivide_18) {period=6.2832})
   B=MaterialFunctionCall(Multiplier=Constant() {r=2}) {materialFunction=&Pi}.Result
  ) {desc=Single Cone intersection area}
  FunctionOutput(None=OneMinus(None=Divide(A=@n13, B=@n14) {constB=1})) {
   description=Returns 1 when there was no occlusion of the light cone and 0 when there is full occlusion of the light cone.
   outputName=Illumination
  }
  FunctionOutput(None=@n13) {
   description=Returns 1 when there was no occlusion of the light cone and 0 when there is full occlusion of the light cone.
   outputName=Area of intersection
   sortPriority=1
  }
 ]
}
