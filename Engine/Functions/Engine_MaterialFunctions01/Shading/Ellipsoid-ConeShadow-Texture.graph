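# Material function that returns soft cone-shadow occlusion at a pixel from an ellipsoid occluder, using a precomputed texture for realtime use.
# It maps the pixel offset into scaled ellipsoid space, then looks up the texture by light-direction cosine and radius-over-distance.
# Static switches choose a single channel, a light-source-angle blend of packed channels, or analytical Spherical-Cap-Intersection for debugging.
# A directional-light switch skips transforming the light vector to save instructions.
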MaterialFunction /Engine/Functions/Engine_MaterialFunctions01/Shading/Ellipsoid-ConeShadow-Texture
graph {
 properties {
  bExposeToLibrary=true
  description=Returns the occlusion value for a pixel given a sphere at a specified position and of a specified radius. The occlusion values are read from a pre-calculated texture so this function is designed to run in realtime.
  libraryCategoriesText [Misc,Shading]
 }
 nodes [
  n1 = FunctionInput() {
   bUsePreviewValueAsDefault=true
   description=" SThe Vector where the directional light is coming from. Should be un-normalized if generated from point light (WorldPosition - LightPosition)."
   inputName=Light Vector
   previewValue=<0 1 -1 1>
  }
  n2 = Normalize(VectorInput=@n1)
  n3 = FunctionInput() {
   bUsePreviewValueAsDefault=true
   description=Supply the axes of the ellipsoid transform in order to orient it correctly in the world.
   inputName=Ellipsoid X Axis
   sortPriority=6
   previewValue=<1 0 0 1>
  }
  n4 = FunctionInput() {
   bUsePreviewValueAsDefault=true
   description=Supply the axes of the ellipsoid transform in order to orient it correctly in the world.
   inputName=Ellipsoid Y Axis
   sortPriority=7
   previewValue=<0 1 0 1>
  }
  n5 = FunctionInput() {
   bUsePreviewValueAsDefault=true
   description=Supply the axes of the ellipsoid transform in order to orient it correctly in the world.
   inputName=Ellipsoid Z Axis
   sortPriority=8
   previewValue=<0 1 0 1>
  }
  n6 = FunctionInput(Preview=StaticBool()) {
   description="If the light is directional, the light vector will not have to be transformed into ellipsoid space which will save considerable instructions."
   inputName=Is Light Directional
   inputType=FunctionInput_StaticBool
   sortPriority=10
  }
  n7 = FunctionInput() {
   bUsePreviewValueAsDefault=true
   description=The Position of the Sphere in the World.
   inputName=Ellipsoid Position
   sortPriority=1
   previewValue=<0 0 100 1>
  }
  n8 = WorldPosition()
  n9 = FunctionInput() {
   bUsePreviewValueAsDefault=true
   description=This is used to scale a sphere into a ellipsoid. Scaling is always along the Ellipsoid's local Z axis.
   inputName=Ellipsoid Z Scale
   inputType=FunctionInput_Scalar
   sortPriority=5
   previewValue=<1 0 0 1>
  }
  n10 = Multiply(
   A=MaterialFunctionCall!(VectorToTransform=Subtract(A=@n7, B=@n8.XYZ), BasisX=@n3, BasisY=@n4, BasisZ=@n5) {materialFunction=/Engine/Functions/Engine_MaterialFunctions02/Math/InverseTransformMatrix}.Transformed Vector 3
   B=AppendVector(A=Constant2Vector() {g=1, r=1}, B=@n9)
  )
  MaterialExpressionDotProduct_4 = DotProduct(
   A=StaticSwitch(
    True=@n2
    False=MaterialFunctionCall(VectorToTransform=@n2, BasisX=@n3, BasisY=@n4, BasisZ=@n5) {materialFunction=&InverseTransformMatrix}.Transformed Vector 3
    Value=@n6
   )
   B=Multiply(A=Normalize(VectorInput=@n10)) {constB=-1}
  )
  n11 = Custom!(x=@MaterialExpressionDotProduct_4) {
   code="return acos(x);"
   description="acos(x)"
   outputType=CMOT_Float1
   inputs [{
     inputName=x
     input {expression=&self:MaterialExpressionDotProduct_4}
    }]
  }
  n12 = FunctionInput() {
   bUsePreviewValueAsDefault=true
   description=The ellipsoid is always scaled along a local Z. The X and Y size need to be the same.
   inputName=Ellipsoid XY size
   inputType=FunctionInput_Scalar
   sortPriority=2
   previewValue=<50 0 0 1>
  }
  n13 = Distance(A=@n10, B=Constant()) {desc=Distance to Sphere}
  MaterialExpressionDivide_24 = Divide(A=@n12, B=@n13) {desc="cos(x)"}
  n14 = Custom(x=@MaterialExpressionDivide_24) {
   code="return asin(x);"
   description="asin(x)"
   outputType=CMOT_Float1
   inputs [{
     inputName=x
     input {expression=&self:MaterialExpressionDivide_24}
    }]
  }
  AppendVector(A=OneMinus(None=Divide(A=@n11) {constB=3.14}), B=Divide(A=@n14) {constB=1.57}) {desc=Linear Version}
  ComponentMask() {b=true, g=true, r=true}
  Custom!(None) {code="float3 colors[10];\r\n\r\ncolors[0]=float3(0,1,0.127);\r\ncolors[1]=float3(0,1,0);\r\ncolors[2]=float3(0.046,0.52,0);\r\ncolors[3]=float3(0.215,0.215,0);\r\ncolors[4]=float3(0.52,0.046,0);\r\ncolors[5]=float3(0.7,0,0);\r\ncolors[6]=float3(1,0,0);\r\ncolors[7]=float3(1,0,0.5);\r\ncolors[8]=float3(1,0.9,0.9);\r\ncolors[9]=float3(1,1,1);\r\n\r\nSteps/=MaxSteps-1;\r\nSteps=saturate(Steps);\r\n\r\nfloat cindex = floor((Steps)*8);\r\nfloat cphase = frac((Steps)*8);\r\n\r\nreturn lerp(colors[cindex],colors[cindex+1],cphase);"}
  n15 = Constant()
  n16 = Custom!(x=@MaterialExpressionDotProduct_4) {
   code="return acos(x);"
   description="acos(x)"
   outputType=CMOT_Float1
   inputs [{
     inputName=x
     input {expression=&self:MaterialExpressionDotProduct_4}
    }]
  }
  n17 = Custom(x=@MaterialExpressionDivide_24) {
   code="return asin(x);"
   description="asin(x)"
   outputType=CMOT_Float1
   inputs [{
     inputName=x
     input {expression=&self:MaterialExpressionDivide_24}
    }]
  }
  n18 = Constant() {r=57.2958}
  n19 = FunctionInput() {
   bUsePreviewValueAsDefault=true
   description="This is used when using the static switch option to blend between channel packed light source textures. When using debug analytical mode, this value is a literal light source angle."
   inputName=Light Source Angle
   inputType=FunctionInput_Scalar
   sortPriority=4
  }
  n20 = MaterialFunctionCall!(
   Angle Between Cones=Multiply(A=@n16, B=Constant() {r=57.2958})
   View Occlusion Angle=Clamp(None=Multiply(A=@n17, B=@n18)) {maxDefault=89}
   Light Source Angle=@n19
   Use Smoothstep=StaticBoolParameter() {parameterName=Use Smoothstep for Area calculation}
   Use Radians
  ) {materialFunction=&/Spherical-Cap-Intersection}
  n21 = ConstantBiasScale(None=@MaterialExpressionDotProduct_4)
  n22 = TextureSample(
   UVs=Clamp(None=AppendVector(A=@n21, B=@MaterialExpressionDivide_24))
   Tex=TextureObjectParameter() {parameterName=Shadow, texture=/Engine/EngineMaterials/T_SphereShadow_RGBA_5-15-30-60deg}
  ) {samplerSource=SSM_Clamp_WorldGroupSettings, texture=&T_SphereShadow_RGBA_5-15-30-60deg}
  n23 = ScalarParameter() {parameterName=Min Texture Lightsource Angle}
  n24 = Subtract(A=@n19, B=@n23) {constB=15}
  n25 = Subtract(A=ScalarParameter() {parameterName=MaxTexture Lightsource Angle}, B=@n23)
  n26 = Multiply(A=Clamp(None=Divide(A=@n24, B=@n25) {constB=60})) {constB=3}
  n27 = Subtract(A=@n26)
  n28 = LinearInterpolate(
   A=LinearInterpolate(A=LinearInterpolate(A=@n22.R, B=@n22.R, Alpha=@n26), B=@n22.B, Alpha=Clamp(None=@n27))
   B=@n22.A
   Alpha=Clamp(None=Subtract(A=@n26) {constB=2})
  )
  n29 = AppendVector(A=@n22.R, B=@n22.R)
  n30 = StaticComponentMaskParameter(None=AppendVector(A=AppendVector(A=@n29, B=@n22.B), B=@n22.A)) {
   defaultR=true
   desc=Which channel of the sphere cone shadow texture to use when not using channel packed light source angle.
   parameterName=Texture Color Channel
  }
  n31 = StaticBoolParameter() {
   desc=When true you can see the shadow as a mathematical area intersection rather than the baked texture result. Can be useful for testing purposes. The mathematical version has pecision errors with low lightsource angles.
   parameterName=Blend Channel Packed Light Source Angles
  }
  n32 = StaticBoolParameter() {
   desc=When true you can see the shadow as a mathematical area intersection rather than the baked texture result. Can be useful for testing purposes. The mathematical version has pecision errors with low lightsource angles.
   parameterName=Debug Using Analytical Calculation
  }
  FunctionOutput(None=StaticSwitch(
    True=If(A=@n12, B=@n13, "A > B"=@n15, "A == B"=@n15, "A < B"=@n20.Illumination)
    False=StaticSwitch(True=@n28, False=@n30, Value=@n31)
    Value=@n32
   ))
  n33 = Constant3Vector() {constant=<rgba 1 0 0 0>}
  n34 = Constant3Vector() {constant=<rgba 0 1 0 0>}
  n35 = Multiply(A=ComponentMask(None=TextureCoordinate()) {g=true}) {constB=3}
  n36 = Subtract(A=@n35)
  LinearInterpolate(
   A=LinearInterpolate(A=LinearInterpolate(A=@n33, B=@n34, Alpha=@n35), B=Constant3Vector() {constant=<rgba 0 0 1 0>}, Alpha=Clamp(None=@n36))
   B=Constant3Vector()
   Alpha=Max(A=Subtract(A=@n35) {constB=2}) {constB=0}
  )
 ]
}
