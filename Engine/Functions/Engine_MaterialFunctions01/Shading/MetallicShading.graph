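# MetallicShading is a library material function that adds view-dependent variation to a metal's base color.
# It clamps the dot product of the world-space pixel normal and camera vector to get a facing term.
# A Fresnel-style power of the inverted term brightens edges, while a powered facing term blends from dim to bright toward view-facing areas.
# The sum of both multiplies the BaseColor input to form the output.
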MaterialFunction /Engine/Functions/Engine_MaterialFunctions01/Shading/MetallicShading
graph {
 properties {bExposeToLibrary=true, description=Apply to base color of metals for more interesting shading, libraryCategoriesText [Misc,Shading]}
 nodes [
  n1 = Clamp(None=DotProduct(A=PixelNormalWS(), B=CameraVectorWS()))
  n2 = Power(Base=OneMinus(None=@n1)) {constExponent=4}
  n3 = LinearInterpolate(Alpha=Power(Base=@n1)) {constA=0.3, constB=0.9}
  FunctionOutput(None=Multiply(A=FunctionInput() {inputName=BaseColor, previewValue=<1 1 1 1>}, B=Add(A=@n2, B=@n3)))
 ]
}
