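# Material function, exposed to the library under Misc and Shading, that computes ambient occlusion cast by a sphere onto the shaded world position.
# Source Radius over distance to Sphere Position, limited to one inside the sphere, gives the half-angle sine, which a Pythagorean helper converts to cosine.
# The resulting spherical cap area is normalized by hemisphere area, clamped and inverted as AO Value.
# Surface normal is not used, so occlusion ignores orientation.
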MaterialFunction /Engine/Functions/Engine_MaterialFunctions01/Shading/Sphere_AO
graph {
 properties {
  bExposeToLibrary=true
  description=This function can be used to calculate the Ambient Occlusion for a sphere affecting the world at any point. It is computed by getting the area of the hemisphere occluded by the occluder.
  libraryCategoriesText [Misc,Shading]
 }
 nodes [
  n1 = MaterialFunctionCall(Multiplier=Constant() {r=2}) {materialFunction=&Pi}
  n2 = FunctionInput() {
   bUsePreviewValueAsDefault=true
   description=This is the size of the light source radius in world units.
   inputName=Source Radius
   inputType=FunctionInput_Scalar
   sortPriority=1
   previewValue=<100 0 0 1>
  }
  n3 = FunctionInput() {
   description=The light position to calculate light source angle for. For a directional light the light source angle is always the same.
   inputName=Sphere Position
  }
  n4 = Distance(A=@n3, B=WorldPosition().XYZ)
  n5 = MaterialFunctionCall!(Side 1 Length=Divide(A=@n2, B=Max(A=@n2, B=@n4)) {desc="sin(x)"}) {materialFunction=/Engine/Functions/Engine_MaterialFunctions02/Math/pythagoreanTheorum}
  n6 = Divide(A=Subtract(A=@n1.Result, B=Multiply(A=@n1.Result, B=@n5.Side 2 Length)) {desc=Area of cap}, B=@n1.Result) {desc=Integrated over Hemisphere}
  FunctionOutput(None=OneMinus(None=Clamp(None=@n6))) {outputName=AO Value}
 ]
}
