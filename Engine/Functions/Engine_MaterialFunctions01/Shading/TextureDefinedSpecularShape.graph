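# Material function that shapes a specular highlight with a texture, described as aligning it to reflected light, though its LightVector sum is left unconnected.
# Camera vector minus the Normal input, remapped by half scale and offset, supplies clamped UVs for the HighlightShape texture's red channel.
# SpecularSharpness widens the lerp range applied to that sample, raising contrast before a clamp.
# The result is multiplied by SpecularIntensity.
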MaterialFunction /Engine/Functions/Engine_MaterialFunctions01/Shading/TextureDefinedSpecularShape
graph {
 properties {description=Aligns the Highlight shape to the reflected light, libraryCategoriesText [Misc,Shading]}
 nodes [
  n1 = CameraVectorWS()
  Add(A=@n1, B=LightVector())
  n2 = Constant()
  n3 = Constant() {r=-1}
  n4 = FunctionInput() {description=Goes from 0-10, inputName=SpecularSharpness, inputType=FunctionInput_Scalar, sortPriority=3, previewValue=<0.5 0 0 1>}
  n5 = Clamp(None=@n4) {maxDefault=100}
  n6 = Constant() {r=0.5}
  n7 = Subtract(A=@n1, B=FunctionInput() {inputName=Normal, sortPriority=1, previewValue=<0 0 1 1>})
  n8 = ComponentMask(None=Add(A=Multiply(A=@n6, B=@n7), B=Constant() {r=0.5})) {g=true, r=true}
  n9 = TextureObject() {texture=/Engine/Engine_MI_Shaders/T_Specular}
  n10 = LinearInterpolate(
   A=Clamp(None=LinearInterpolate(A=@n2, B=@n3, Alpha=@n5)) {maxDefault=0, minDefault=-100}
   B=LinearInterpolate(A=Constant() {r=1}, B=Constant() {r=1.5}, Alpha=@n5)
   Alpha=TextureSample(
    UVs=Clamp(None=@n8)
    Tex=FunctionInput(Preview=@n9) {description=Should have clamped address and use the red channel, inputName=HighlightShape, inputType=FunctionInput_Texture2D}
   ).R
  )
  FunctionOutput(None=Multiply(A=Clamp(None=@n10), B=FunctionInput() {inputName=SpecularIntensity, sortPriority=2}))
 ]
}
