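# Material function that converts DiffuseColor and SpecularColor inputs into BaseColor, Metallic and Specular outputs for the metallic workflow.
# Metallic is a heuristic mask from desaturated inputs, multiplying a dark-diffuse term by a term for specular's share of combined brightness, each clamped.
# BaseColor blends from diffuse toward diffuse plus specular by that mask.
# Specular is the desaturated specular color scaled by a constant.
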MaterialFunction /Engine/Functions/Engine_MaterialFunctions01/Shading/ConvertFromDiffSpec
graph {nodes [
  n1 = FunctionInput() {bUsePreviewValueAsDefault=true, inputName=DiffuseColor}
  n2 = FunctionInput() {bUsePreviewValueAsDefault=true, inputName=SpecularColor, sortPriority=1}
  n3 = Desaturation(None=@n1)
  n4 = Add(A=@n3) {constB=-0.4}
  n5 = Desaturation(None=@n2)
  n6 = Add(A=@n3, B=@n5)
  n7 = Constant() {r=0.5}
  n8 = Multiply(A=Subtract(A=Divide(A=@n5, B=@n6), B=@n7)) {constB=4}
  n9 = Multiply(A=Clamp(None=Multiply(A=@n4) {constB=-5}), B=Clamp(None=Add(A=@n8, B=@n7)))
  FunctionOutput(None=LinearInterpolate(A=@n1, B=Add(A=@n1, B=@n2), Alpha=@n9)) {outputName=BaseColor}
  FunctionOutput(None=@n9) {outputName=Metallic, sortPriority=1}
  FunctionOutput(None=Multiply(A=@n5) {constB=12.5}) {outputName=Specular, sortPriority=2}
 ]}
