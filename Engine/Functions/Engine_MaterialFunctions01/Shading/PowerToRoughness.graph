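# PowerToRoughness is a material function that converts a scalar specular-style Power exponent into a Roughness value for shading.
# It clamps Power to a safe positive range, then evaluates 2 / (4 * Power + 2).
# Two chained square roots take the fourth root of that result, so higher Power yields lower Roughness.
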MaterialFunction /Engine/Functions/Engine_MaterialFunctions01/Shading/PowerToRoughness
graph {nodes [
  n1 = Clamp(None=FunctionInput() {inputName=Power, inputType=FunctionInput_Scalar}) {maxDefault=200000, minDefault=0.25}
  n2 = Divide(B=Add(A=Multiply(A=@n1) {constB=4}) {constB=2}) {constA=2}
  FunctionOutput(None=SquareRoot(None=SquareRoot(None=@n2))) {outputName=Roughness}
 ]}
